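# Sound cue for the FNCS Trophy Celebration emote audio, layering music and SFX intros that hand off to looping versions.
# Both intros play immediately, while a fixed delay of about 5.6 seconds holds back the two loops.
# Each loop has its own Enveloper volume curve to fade in under a second, with the SFX loop mixed lower than music.
# It uses first-person emote music class and concurrency, and references a music-analysis submix send.
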SoundCue /BRCosmetics/Sounds/Emotes/FNCS_Trophy_Celebration/Emote_FNCS_TrophyCelebration_Cue
root=Mixer_1
nodes [
 {id=Delay_1, children [Mixer_2], properties {delayMax=5.597, delayMin=5.597}}
 {
  id=Enveloper_1
  children [WavePlayer_3]
  properties {
   pitchCurve {editorCurveData {keys [{time=0}]}}
   volumeCurve {editorCurveData {keys [<at 0 0.01>,<at 0.1772 0.039>,<at 0.3951 0.1723>,<at 0.5686 0.3886>,<at 0.6997 0.7097>,<at 0.751 1.0031>]}}
  }
 }
 {
  id=Enveloper_2
  children [WavePlayer_1]
  properties {
   pitchCurve {editorCurveData {keys [{time=0}]}}
   volumeCurve {editorCurveData {keys [<at 0 0.01>,<at 0.2651 0.0232>,<at 0.379 0.0358>,<at 0.4699 0.0899>,<at 0.5483 0.1351>,<at 0.6475 0.2433>,<at 0.7383 0.3516>,<at 0.8292 0.5501>,<at 0.9076 0.7215>,<at 0.9531 1>]}}
  }
 }
 {id=Mixer_1, children [Mixer_3,Delay_1], properties {inputVolume [1.2,1]}}
 {id=Mixer_2, children [Enveloper_2,Enveloper_1], properties {inputVolume [0.65,1]}}
 {id=Mixer_3, children [WavePlayer_4,WavePlayer_5], properties {inputVolume [1,1]}}
 {
  id=WavePlayer_1
  properties {bLooping=true, soundWaveAssetPtr=&/Emote_FNCS_Celebration_SFX_Loop}
 }
 {
  id=WavePlayer_3
  properties {bLooping=true, soundWaveAssetPtr=&/Emote_FNCS_Celebration_Music_Loop}
 }
 {id=WavePlayer_4, properties {soundWaveAssetPtr=&/Emote_FNCS_Celebration_MX_Intro}}
 {id=WavePlayer_5, properties {soundWaveAssetPtr=&/Emote_FNCS_Celebration_SFX_Intro}}
]
properties {
 maxDistance=2097152
 soundClassObject=/Game/Sounds/FortSoundClassesAndModes/Fort_Music_Emote_1P_PSM
 volumeMultiplier=0.9
 concurrencyOverrides {maxCount=3, resolutionRule=StopOldest}
 concurrencySet [/Game/Sounds/Concurrency_Rules/Athena_EmoteMusic_1P_Conc]
 soundSubmixSends [{minSendLevel=0, soundSubmix=&Fortnite_AudioAnalysis_Music}]
}
