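# Third-person sound cue for the FNCS Trophy Celebration emote, layering music, effects and crowd applause.
# Music and slightly delayed SFX intros play first, with one-shot applause delayed to about five seconds.
# Looping music and SFX start shortly afterward, using nested volume envelopes to fade in beneath the intro's tail.
# Emote attenuation and concurrency apply, playback continues when silent, and a submix send targets music audio analysis.
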SoundCue /BRCosmetics/Sounds/Emotes/FNCS_Trophy_Celebration/SC_Emote_FNCS_TrophyCelebration_3P_Cue
root=Mixer_0
nodes [
 {id=Delay_0, children [Mixer_1], properties {delayMax=5.596, delayMin=5.596}}
 {id=Delay_1, children [WavePlayer_0], properties {delayMax=5.2, delayMin=5.2}}
 {id=Delay_2, children [WavePlayer_5], properties {delayMax=0.2, delayMin=0.2}}
 {
  id=Enveloper_1
  children [Delay_0]
  properties {
   pitchCurve {editorCurveData {keys [{time=0}]}}
   volumeCurve {editorCurveData {keys [<at 0 0.01>,<at 0.9074 0.0517>,<at 1.5791 0.2147>,<at 2.0506 0.5814>,<at 2.25 1>]}}
  }
 }
 {
  id=Enveloper_2
  children [WavePlayer_1]
  properties {pitchCurve {editorCurveData {keys [{time=0}]}}, volumeCurve {editorCurveData {keys [<at 0 0.01>,<at 0.6585 0.4845>,<at 1 1>]}}}
 }
 {id=Mixer_0, children [Mixer_2,Enveloper_1,Delay_1], properties {inputVolume [1.2,1,1]}}
 {id=Mixer_1, children [Enveloper_2,WavePlayer_3], properties {inputVolume [1.4,0.4]}}
 {id=Mixer_2, children [WavePlayer_4,Delay_2], properties {inputVolume [0.5,1]}}
 {id=WavePlayer_0, properties {soundWaveAssetPtr=&/Emote_FNCS_Crowd_Applause}}
 {
  id=WavePlayer_1
  properties {bLooping=true, soundWaveAssetPtr=&/Emote_FNCS_Celebration_SFX_Loop}
 }
 {
  id=WavePlayer_3
  properties {bLooping=true, soundWaveAssetPtr=&/Emote_FNCS_Celebration_Music_Loop}
 }
 {id=WavePlayer_4, properties {soundWaveAssetPtr=&/Emote_FNCS_Celebration_MX_Intro}}
 {id=WavePlayer_5, properties {soundWaveAssetPtr=&/Emote_FNCS_Celebration_SFX_Intro}}
]
properties {
 attenuationSettings=&Fort_Emotes_Athena_Att
 maxDistance=1600
 soundClassObject=/Game/Sounds/FortSoundClassesAndModes/Fort_Music_Emote_3P
 virtualizationMode=PlayWhenSilent
 volumeMultiplier=0.9
 concurrencyOverrides {resolutionRule=PreventNew}
 concurrencySet [/Game/Sounds/Concurrency_Rules/Athena_EmoteMusic_Conc]
 soundSubmixSends [{minSendLevel=0, soundSubmix=&Fortnite_AudioAnalysis_Music}]
}
